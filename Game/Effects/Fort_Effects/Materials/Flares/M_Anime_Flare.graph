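# Translucent unlit flare material for sprite, mesh and ribbon particles, using an anime flare texture by default.
# Particle color tints a selectable texture channel, which feeds HDR-from-LDR remapping and shared emissive adjustments.
# Alpha combines particle alpha with that channel, or a separately selected mask channel when UseMaskselect exceeds the comparison value.
# A parameterized near-camera fade also scales alpha, except at low quality, where it becomes a constant one.
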Material /Game/Effects/Fort_Effects/Materials/Flares/M_Anime_Flare
graph {
 properties {
  bEnableResponsiveAA=true
  bUsedWithNiagaraMeshParticles=true
  bUsedWithNiagaraRibbons=true
  bUsedWithNiagaraSprites=true
  bUsedWithParticleSprites=true
  blendMode=BLEND_Translucent
  pixelDepthOffsetMode=PDOM_Legacy
  previewMesh=/Engine/Transient.MaterialEditorMeshComponent_41
  shadingModel=MSM_Unlit
  usedShadingModels=Unlit
 }
 nodes [
  n1 = ParticleColor()
  n2 = TextureSampleParameter2D() {parameterName=Flare_Texture, texture=/Game/Effects/Fort_Effects/Textures/Flares/T_Anime_Flare}
  n3 = ChannelMaskParameter(None=@n2.RGBA) {group=Base, maskChannel=Green, parameterName=FlareTexture_RGBselect, sortPriority=5, defaultValue=<rgba 0 1 0 0>}
  n4 = Multiply(A=@n1.RGB, B=@n3.R)
  MaterialFunctionCall(Emissive=MaterialFunctionCall(LDR Input=Clamp(None=@n4), HDR Tint=@n4) {materialFunction=&DeriveHDRfromLDR}.RemappedRange) {materialFunction=&MF_FX_EmissiveAdjustments}
  n5 = ScalarParameter() {parameterName=UseMaskselect}
  n6 = ChannelMaskParameter(None=@n2.RGBA) {group=Base, maskChannel=Green, parameterName=FlareTexture_Maskselect, sortPriority=5, defaultValue=<rgba 0 1 0 0>}
  n7 = MaterialFunctionCall(Distance=ScalarParameter() {defaultValue=200, parameterName=CameraDistanceFade}) {materialFunction=&Near_Camera_Fade}
  Multiply(
   A=Multiply(A=@n1.A, B=If(A=@n5, "A > B"=@n6.R, "A == B"=@n3.R, "A < B"=@n3.R))
   B=QualitySwitch(Default=@n7.Result, Low=Constant() {r=1}, High=@n7.Result, Medium=@n7.Result)
  )
 ]
}
