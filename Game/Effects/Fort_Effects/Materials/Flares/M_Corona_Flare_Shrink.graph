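# Additive unlit corona flare material for particle and Niagara sprites that shrinks as the camera nears it.
# A sphere mask indexes a halo gradient, added to hot-center glows and a cross flare mask, tinted by particle color.
# Multiplied sine waves, blended by dynamic parameters, drive flicker in corona size and brightness.
# Depth and near-camera fades soften it, while a distance-based offset toward particle center collapses the sprite, except on mobile.
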Material /Game/Effects/Fort_Effects/Materials/Flares/M_Corona_Flare_Shrink
graph {
 properties {
  bUsedWithNiagaraSprites=true
  bUsedWithParticleSprites=true
  blendMode=BLEND_Additive
  pixelDepthOffsetMode=PDOM_Legacy
  shadingModel=MSM_Unlit
  usedShadingModels=Unlit
 }
 nodes [
  n1 = Constant() {r=0.5}
  n2 = SphereMask(A=TextureCoordinate(), B=@n1, Radius=@n1) {hardnessPercent=0}
  n3 = Constant() {r=0.5}
  n4 = Multiply(A=Time(), B=Constant() {r=1})
  n5 = Sine(None=@n4) {period=0.075}
  n6 = Multiply(A=Multiply(A=Sine(None=@n4) {period=0.155}, B=Sine(None=@n4) {period=0.2}), B=Sine(None=@n4) {period=0.43})
  n7 = DynamicParameter() {paramNames [FlickerFade,CoronaSize,Param3,Size]}
  n8 = LinearInterpolate(A=Clamp(None=Multiply(A=@n5, B=@n6)) {maxDefault=3}, Alpha=@n7.FlickerFade) {constA=1}
  n9 = LinearInterpolate(Alpha=@n8) {constA=-0.4, constB=0}
  n10 = SphereMask(
   A=TextureCoordinate()
   B=@n3
   Radius=@n3
   Hardness=LinearInterpolate(Alpha=Add(A=@n9, B=@n7.CoronaSize) {constB=0}) {constA=0.7, constB=-0.3}
  ) {hardnessPercent=25}
  n11 = Add(
   A=Multiply(A=Power(Base=@n2) {constExponent=25}) {constB=2}
   B=Multiply(
    A=TextureSample(UVs=@n10) {texture=/Game/Effects/Fort_Effects/Textures/Grad/T_Halo_Corona_Gradient}.RGB
    B=ScalarParameter() {defaultValue=0.125, parameterName=CoronaBrightness}
   ) {constB=0.125}
  )
  n12 = ParticleColor()
  n13 = LinearInterpolate(B=@n12.RGB) {constA=1, constAlpha=0.7}
  n14 = Add(
   A=Multiply(A=Power(Base=@n2) {constExponent=4}) {constB=0.15}
   B=TextureSampleParameter2D() {parameterName=FlareMask, samplerType=SAMPLERTYPE_Grayscale, texture=/Game/Effects/Fort_Effects/Textures/Flares/T_Cross_Flare}.RGB
  )
  MaterialFunctionCall(Emissive=Add(A=Multiply(A=@n11, B=@n13), B=Multiply(A=@n14, B=@n12.RGB))) {materialFunction=&MF_FX_EmissiveAdjustments}
  n15 = DepthFade(Opacity=@n12.A, FadeDistance=ScalarParameter() {defaultValue=250, parameterName=DepthFade})
  n16 = MaterialFunctionCall(Distance=ShadingPathSwitch(
    Default=ScalarParameter() {defaultValue=50, parameterName=NearCameraShrink}
    Mobile=ScalarParameter() {defaultValue=250, parameterName=NearCameraShrink_Mobile}
   ) {desc=Migrated from FeatureLevelSwitch}) {materialFunction=&Near_Camera_Fade}
  Multiply(
   A=@n8
   B=Multiply(
    A=ShadingPathSwitch(Default=@n15, Deferred=@n15, Mobile=@n12.A) {desc=Migrated from FeatureLevelSwitch}
    B=Power(Base=@n16.Result) {constExponent=3}
   )
  )
  n17 = ParticlePositionWS()
  n18 = WorldPosition()
  n19 = CameraPositionWS()
  n20 = Divide(
   A=Subtract(A=Distance(A=@n17, B=@n19), B=ScalarParameter() {parameterName=DistanceOffset_1})
   B=ScalarParameter() {defaultValue=1000, parameterName=ScaleUpDistance_1}
  )
  ShadingPathSwitch(Default=LinearInterpolate(A=Subtract(A=@n17, B=@n18.XYZ), Alpha=Clamp(None=@n20)) {constB=0}, Mobile=Constant()) {desc=Migrated from FeatureLevelSwitch}
 ]
}
