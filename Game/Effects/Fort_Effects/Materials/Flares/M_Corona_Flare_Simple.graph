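# Additive, unlit, two-sided flickering corona flare material for sprites, mesh particles and skeletal meshes.
# Radial masks build a core, edge and gradient-textured halo, combined with a spike texture, tinted and scaled by time of day.
# Multiplied sines drive flicker in halo size and opacity, controlled by per-particle random on GPU particles or dynamic parameters otherwise.
# Opacity adds depth fade, near-camera fade and an optional flare mask.
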Material /Game/Effects/Fort_Effects/Materials/Flares/M_Corona_Flare_Simple
graph {
 properties {
  bUsedWithMeshDeformer=true
  bUsedWithNiagaraMeshParticles=true
  bUsedWithNiagaraSprites=true
  bUsedWithParticleSprites=true
  bUsedWithSkeletalMesh=true
  blendMode=BLEND_Additive
  pixelDepthOffsetMode=PDOM_Legacy
  shadingModel=MSM_Unlit
  twoSided=true
  usedShadingModels=Unlit
 }
 nodes [
  n1 = ScalarParameter() {defaultValue=1, group=Time of Day Scale Values, parameterName=Morning / Evening Scale Value}
  n2 = ScalarParameter() {defaultValue=1.1, group=Time of Day Scale Values, parameterName=Day Scale Value}
  n3 = ScalarParameter() {defaultValue=0.3, group=Time of Day Scale Values, parameterName=Night Scale Value}
  n4 = Constant() {r=0.5}
  n5 = SphereMask(A=TextureCoordinate(), B=@n4, Radius=@n4) {hardnessPercent=0}
  n6 = Multiply(A=Power(Base=@n5) {constExponent=25}, B=ScalarParameter() {defaultValue=2, parameterName=CenterBrightness}) {constB=2}
  n7 = TextureCoordinate()
  n8 = Constant() {r=0.5}
  n9 = ParticleRandom()
  n10 = Multiply(
   A=Time()
   B=StaticSwitchParameter(True=LinearInterpolate(Alpha=@n9) {constA=1.2, constB=0.8}, False=Constant() {r=1}) {parameterName=GPU Particles}
  )
  n11 = Sine(None=@n10) {period=0.075}
  n12 = Multiply(A=Multiply(A=Sine(None=@n10) {period=0.155}, B=Sine(None=@n10) {period=0.2}), B=Sine(None=@n10) {period=0.43})
  n13 = ScalarParameter() {defaultValue=0.4, desc="0 = full flicker amount", parameterName=GPUFlickerFadeMin}
  n14 = ScalarParameter() {defaultValue=1, desc="1 = no flicker", parameterName=GPUFlickerFadeMax, sortPriority=1}
  n15 = ParticleRandom()
  n16 = ScalarParameter() {defaultValue=1, parameterName=FlickerLerp}
  n17 = DynamicParameter() {paramNames [FlickerFade,CoronaSize,Param3,Param4]}
  n18 = LinearInterpolate(
   A=Clamp(None=Multiply(A=@n11, B=@n12)) {maxDefault=3}
   Alpha=StaticSwitchParameter(True=LinearInterpolate(A=@n13, B=@n14, Alpha=@n15) {constA=0.5}, False=Multiply(A=@n16, B=@n17.FlickerFade)) {parameterName=GPU Particles}
  ) {constA=1}
  n19 = ScalarParameter() {defaultValue=1, parameterName=CoronaSize}
  n20 = Multiply(A=@n17.CoronaSize, B=@n19)
  n21 = LinearInterpolate(Alpha=Add(A=LinearInterpolate(Alpha=@n18) {constA=-0.4, constB=0}, B=StaticSwitchParameter(True=@n19, False=@n20) {parameterName=GPU Particles}) {constB=0}) {constA=0.7, constB=-0.3}
  n22 = Multiply(
   A=TextureSample(UVs=SphereMask(A=@n7, B=@n8, Radius=@n8, Hardness=@n21) {hardnessPercent=25}) {texture=/Game/Effects/Fort_Effects/Textures/Grad/T_Halo_Corona_Gradient}.RGB
   B=ScalarParameter() {defaultValue=0.125, parameterName=CoronaBrightness}
  ) {constB=0.125}
  n23 = VectorParameter() {parameterName=Color, defaultValue=<rgba 1 1 1 1>}
  n24 = ParticleColor()
  n25 = StaticSwitchParameter(True=@n23.RGB, False=Multiply(A=@n24.RGB, B=@n23.RGB)) {parameterName=UsedWithSkeletalMesh}
  n26 = Multiply(A=ShadingPathSwitch(Default=Add(A=@n6, B=@n22), Forward=@n22), B=LinearInterpolate(B=@n25) {constA=1, constAlpha=0.7})
  n27 = Power(Base=@n5) {constExponent=4}
  n28 = ScalarParameter() {defaultValue=0.15, parameterName=EdgeBrightness}
  n29 = Multiply(
   A=Add(
    A=Multiply(A=@n27, B=@n28) {constB=0.15}
    B=TextureSampleParameter2D() {parameterName=FlareMask, texture=/Game/Effects/Fort_Effects/Textures/Flares/T_Flare_Spikeball}.RGB
   )
   B=@n25
  )
  MaterialFunctionCall(Emissive=Multiply(
    A=MaterialFunctionCall(Value_Morning=@n1, Value_Day=@n2, Value_Evening=@n1, Value_Night=@n3) {materialFunction=&Scalar_Time_of_Day}.Result
    B=Add(A=@n26, B=@n29)
   )) {materialFunction=&MF_FX_EmissiveAdjustments}
  n30 = TextureSampleParameter2D() {parameterName=OuterFlareMaskTexture, samplerType=SAMPLERTYPE_Grayscale, texture=/Game/Effects/Fort_Effects/Textures/T_EFX_CameraFlare_01}
  n31 = StaticSwitchParameter(True=@n23.A, False=Multiply(A=@n24.A, B=@n23.A)) {parameterName=UsedWithSkeletalMesh}
  n32 = DepthFade(Opacity=@n31, FadeDistance=ScalarParameter() {defaultValue=20, parameterName=DepthFade})
  n33 = MaterialFunctionCall(Distance=ScalarParameter() {defaultValue=150, parameterName=NearCameraShrink}) {materialFunction=&Near_Camera_Fade}
  n34 = Multiply(
   A=@n18
   B=Multiply(
    A=ShadingPathSwitch(Default=@n32, Deferred=@n32, Mobile=@n31) {desc=Migrated from FeatureLevelSwitch}
    B=Power(Base=@n33.Result) {constExponent=3}
   )
  )
  StaticSwitchParameter(True=Multiply(A=@n30.G, B=@n34), False=@n34) {parameterName=UseFlareMask}
 ]
}
