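# Unlit, two-sided Modulate material for particle and Niagara sprites that, by its name, darkens the scene behind a soft radial gradient.
# A sphere mask, shaped by a dynamic Falloff exponent, scales particle color and alpha with depth and near-camera fades.
# An optional world-position-offset branch shrinks sprites toward their particle position as the camera nears.
# Mobile skips both fades and substitutes object position for particle position.
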Material /Game/Effects/Fort_Effects/Materials/Flares/M_Darken_Gradient_Modulate
graph {
 properties {
  bUseTranslucencyVertexFog=false
  bUsedWithNiagaraSprites=true
  bUsedWithParticleSprites=true
  blendMode=BLEND_Modulate
  pixelDepthOffsetMode=PDOM_Legacy
  shadingModel=MSM_Unlit
  translucencyPass=MTP_BeforeDOF
  twoSided=true
  usedShadingModels=Unlit
 }
 nodes [
  n1 = ParticleColor()
  n2 = Constant() {r=0.5}
  n3 = DynamicParameter() {defaultValue=<rgba 3 1 1 1>, paramNames [Falloff,Param2,Param3,Param4]}
  n4 = Power(Base=SphereMask(A=TextureCoordinate(), B=@n2, Radius=@n2) {hardnessPercent=0}, Exp=@n3.Falloff) {constExponent=3}
  n5 = Multiply(A=@n1.RGB, B=@n4)
  n6 = DepthFade(
   Opacity=MaterialFunctionCall(Distance=ScalarParameter() {defaultValue=250, parameterName=NearCameraFadeDist}) {materialFunction=&Near_Camera_Fade}.Result
   FadeDistance=ScalarParameter() {defaultValue=50, parameterName=DepthFade}
  )
  n7 = ShadingPathSwitch(Default=@n6, Deferred=@n6, Mobile=Constant() {r=1}) {desc=Migrated from FeatureLevelSwitch}
  Add(A=Multiply(A=Multiply(A=@n5, B=@n4), B=Multiply(A=@n1.A, B=@n7)))
  Multiply(A=OneMinus(None=@n4), B=@n3.Param2)
  n8 = ParticlePositionWS()
  n9 = ShadingPathSwitch(Default=@n8, Deferred=@n8, Mobile=ObjectPositionWS()) {desc=Migrated from FeatureLevelSwitch}
  n10 = WorldPosition()
  n11 = Divide(A=Distance(A=@n9, B=CameraPositionWS()), B=ScalarParameter() {defaultValue=250, parameterName=NearFadeDistance})
  StaticSwitchParameter(True=LinearInterpolate(A=Subtract(A=@n9, B=@n10.XYZ), Alpha=Clamp(None=@n11)) {constB=0}, False=Constant()) {parameterName=SizeoftheCameraWPO}
 ]
}
