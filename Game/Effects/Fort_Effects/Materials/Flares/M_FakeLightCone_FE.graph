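# Additive, unlit, two-sided material that fakes a light cone with foggy shimmer.
# A flippable, powered V gradient times a view-angle fade blends two colors into HDR emissive and drives opacity.
# Two panning textures break up opacity, with fog noise offset by actor position, before depth fade softens intersections.
# Low quality swaps noise and depth fade for dimmed, front-face-only opacity and adds a camera offset.
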Material /Game/Effects/Fort_Effects/Materials/Flares/M_FakeLightCone_FE
graph {
 properties {blendMode=BLEND_Additive, pixelDepthOffsetMode=PDOM_Legacy, shadingModel=MSM_Unlit, twoSided=true, usedShadingModels=Unlit}
 nodes [
  n1 = ComponentMask(None=TextureCoordinate()) {g=true}
  n2 = Power(
   Base=LinearInterpolate(A=@n1, B=OneMinus(None=@n1), Alpha=ScalarParameter() {defaultValue=1, parameterName=FlipY})
   Exp=ScalarParameter() {defaultValue=0.9, parameterName=ConePow}
  )
  n3 = MaterialFunctionCall("FadeSensitivity (optional)"=ScalarParameter() {defaultValue=0.7, parameterName=FadeSensitivity}) {materialFunction=&FadeBasedOnViewAngle}
  n4 = LinearInterpolate(
   A=VectorParameter() {parameterName=LightEmissiveColor}.RGB
   B=VectorParameter() {parameterName=LightEmissiveColor_1, defaultValue=<rgba 0.2483 0.6951 1.3 0>}.RGB
   Alpha=Clamp(None=Multiply(A=@n2, B=@n3.Result))
  )
  MaterialFunctionCall(LDR Input=Clamp(None=@n4), HDR Tint=@n4, Desaturate Luminance=ScalarParameter() {defaultValue=0.15, parameterName=HDRLum}) {materialFunction=&DeriveHDRfromLDR}
  n5 = ScalarParameter() {defaultValue=0.45, parameterName=OpacityScale}
  n6 = Multiply(A=@n2, B=@n3.Result)
  n7 = TextureCoordinate()
  n8 = Multiply(A=ComponentMask(None=ActorPositionWS()) {g=true, r=true}, B=ScalarParameter() {defaultValue=0.0001, parameterName=adsfsaf}) {constB=0.005}
  n9 = Panner(Coordinate=Multiply(A=Add(A=@n7, B=@n8), B=ScalarParameter() {defaultValue=1.25, parameterName=TileFogNoise})) {speedY=0.05}
  n10 = Panner(Coordinate=TextureCoordinate()) {speedX=-0.125}
  n11 = LinearInterpolate(
   A=@n6
   B=Multiply(A=@n6, B=TextureSample(UVs=@n9) {texture=&T_Smooth_tile}.G)
   Alpha=OneMinus(None=TextureSample(UVs=@n10) {texture=/Game/Effects/Fort_Effects/Textures/T_HitImpact}.G)
  )
  n12 = TwoSidedSign()
  QualitySwitch(Default=DepthFade(Opacity=Multiply(A=@n5, B=@n11)), Low=Multiply(A=Multiply(A=@n6) {constB=0.2}, B=Saturate(None=@n12)))
  QualitySwitch(
   Default=Constant3Vector()
   Low=MaterialFunctionCall(Offset Amount=ScalarParameter() {defaultValue=10, parameterName=CameraOffset}, World Position=ObjectPositionWS()) {materialFunction=/Engine/Functions/Engine_MaterialFunctions02/WorldPositionOffset/CameraOffset}.Clamped Camera Offset
  )
  WorldPosition()
 ]
}
