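# Unlit translucent flare material for particle and Niagara sprites, tinted by particle color through an emissive adjustment function.
# An alpha product combines particle alpha, a powered sphere mask and radial flare textures rotating in opposite directions.
# A time-driven ceiled cosine, scaled by dynamic parameter Param2, hard-switches between the two flare channels.
# The TextureFlip switch has no visible effect because both its lerp inputs are the same channel.
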Material /Game/Effects/Fort_Effects/Materials/Flares/M_Flare
graph {
 properties {
  bUsedWithNiagaraSprites=true
  bUsedWithParticleSprites=true
  blendMode=BLEND_Translucent
  pixelDepthOffsetMode=PDOM_Legacy
  shadingModel=MSM_Unlit
  usedShadingModels=Unlit
 }
 nodes [
  n1 = ParticleColor()
  MaterialFunctionCall(Emissive=@n1.RGB) {materialFunction=&MF_FX_EmissiveAdjustments}
  n2 = TextureSample() {texture=/Game/Effects/Fort_Effects/Textures/Flares/T_Sphere_Color_Packed}
  n3 = TextureSample(UVs=Rotator(Coordinate=TextureCoordinate()) {speed=-2}) {texture=/Game/Effects/Fort_Effects/Textures/Flares/T_Flares_Radial_Packed}
  n4 = TextureSample(UVs=Rotator(Coordinate=TextureCoordinate()) {speed=4}) {texture=&T_Flares_Radial_Packed}
  n5 = DynamicParameter() {defaultValue=<rgba 0 0 0 1>, paramNames [TextureFlip,Param2,Param3,Param4]}
  n6 = Multiply(A=Time(), B=@n5.TextureFlip)
  n7 = LinearInterpolate(A=@n4.R, B=@n4.R, Alpha=Ceil(None=Cosine(None=@n6) {period=0.2}))
  n8 = Time()
  n9 = Ceil(None=Cosine(None=Multiply(A=@n8, B=@n5.Param2)) {period=0.1})
  Multiply(A=@n1.A, B=Multiply(A=Power(Base=@n2.G) {constExponent=0.5}, B=LinearInterpolate(A=@n3.G, B=@n7, Alpha=@n9)))
 ]
}
