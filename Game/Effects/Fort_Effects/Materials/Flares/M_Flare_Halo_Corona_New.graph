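# Additive, unlit, two-sided mesh-particle material for an animated halo or corona flare effect.
# Color hue-rotates red by time, UVs and panning noise, then blends toward particle color via the dynamic Tint parameter.
# Opacity multiplies squared panning noise, a curve-atlas falloff along V, particle alpha and view-facing edge fade.
# Switchable world-space noise, partly anchored to actor position, modulates opacity; the deferred path adds depth fade.
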Material /Game/Effects/Fort_Effects/Materials/Flares/M_Flare_Halo_Corona_New
graph {
 properties {
  bUsedWithMeshParticles=true
  bUsedWithNiagaraMeshParticles=true
  blendMode=BLEND_Additive
  pixelDepthOffsetMode=PDOM_Legacy
  shadingModel=MSM_Unlit
  twoSided=true
  usedShadingModels=Unlit
 }
 nodes [
  n1 = TextureCoordinate()
  n2 = Add(A=Multiply(A=Time()) {constB=0.1}, B=ComponentMask(None=@n1) {r=true})
  n3 = WorldPosition()
  n4 = ActorPositionWS()
  n5 = Multiply(A=@n4) {constB=0.85}
  n6 = Add(A=ComponentMask(None=Subtract(A=@n3.XYZ, B=@n5)) {g=true, r=true}, B=ComponentMask(None=Subtract(A=@n3.XYZ, B=@n4)) {b=true})
  n7 = Multiply(
   A=TextureSample(UVs=Divide(A=@n6) {constB=1024}) {texture=/Game/Global/Textures/Default/Noise/T_FireNoise_01}.R
   B=TextureSample(UVs=Divide(A=@n6) {constB=1536}) {texture=&T_FireNoise_01}.R
  )
  Add(A=@n2, B=@n7)
  n8 = ComponentMask(None=@n1) {g=true}
  n9 = ScalarParameter() {defaultValue=1, parameterName=T_PannerSpeedX}
  n10 = ScalarParameter() {defaultValue=1, parameterName=T_PannerSpeedY}
  n11 = Panner(Coordinate=Multiply(A=TextureCoordinate() {vTiling=0.1}, B=AppendVector(A=@n9, B=@n10))) {speedX=0.05, speedY=0.15}
  n12 = TextureSample(UVs=@n11) {texture=&T_FireNoise_01}
  n13 = Add(A=@n8, B=Multiply(A=Multiply(A=@n12.R) {constB=0.5}, B=@n8))
  n14 = Constant3Vector() {constant=<rgba 1 0 0 0>}
  n15 = RotateAboutAxis(
   NormalizedRotationAxis=Multiply(A=Constant3Vector() {constant=<rgba 1 1 1 0>}, B=Constant() {r=0.5774})
   RotationAngle=Add(A=Multiply(A=@n13) {constB=0.5}, B=@n2)
   PivotPoint=Constant()
   Position=@n14
  )
  MaterialFunctionCall(Emissive=LinearInterpolate(A=Add(A=@n15, B=@n14), B=ParticleColor().RGB, Alpha=DynamicParameter() {paramNames [Tint,Param2,Param3,Param4]}.Tint)) {materialFunction=&MF_FX_EmissiveAdjustments}
  n16 = Constant2Vector() {g=0.0618, r=1}
  n17 = Multiply(A=@n12.R, B=TextureSample(UVs=Multiply(A=@n11, B=@n16) {constB=0.5}) {texture=&T_FireNoise_01}.R)
  n18 = TextureCoordinate()
  n19 = CurveAtlasRowParameter(CurveTime=OneMinus(None=ComponentMask(None=@n18) {g=true})) {
   atlas=&AthenaCurveAtlas01
   curve=/Game/Gadgets/Assets/VinderTech_GliderChute/Glider_Tech_Ops/FX/Curve_Glider_Tech_Ops_2
   defaultValue=47
   parameterName=CurveAtlas
  }
  n20 = Multiply(A=Multiply(A=Multiply(A=@n17, B=@n17)) {constB=2}, B=@n19)
  n21 = ParticleColor()
  n22 = DotProduct(A=CameraVectorWS(), B=VertexNormalWS())
  n23 = Multiply(
   A=StaticSwitchParameter(True=LinearInterpolate(Alpha=@n7) {constB=2}, False=Constant() {r=1}) {defaultValue=true, parameterName=UseWPtexture}
   B=Multiply(A=Multiply(A=@n20, B=@n21.A), B=Abs(None=@n22))
  )
  ShadingPathSwitch(Default=@n23, Deferred=DepthFade(Opacity=@n23) {fadeDistanceDefault=15})
  SquareRoot(None=@n7)
 ]
}
