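# Translucent unlit effect material for soft light-shaft flares on sprites and Niagara mesh particles, colored by particle color through an emissive adjustment function.
# Flare texture, with optional random sub-UV, gradient and particle alpha, drives depth-faded opacity with optional camera and view-angle fades.
# Mobile shading path skips all three fades.
# A separate panning radial-noise mask branch does not feed that opacity chain.
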Material /Game/Effects/Fort_Effects/Materials/Flares/M_Flare_LightShafts_Soft
graph {
 properties {
  bAllowNegativeEmissiveColor=true
  bUsedWithNiagaraMeshParticles=true
  bUsedWithNiagaraSprites=true
  bUsedWithParticleSprites=true
  blendMode=BLEND_Translucent
  pixelDepthOffsetMode=PDOM_Legacy
  shadingModel=MSM_Unlit
  usedShadingModels=Unlit
 }
 nodes [
  n1 = ScalarParameter() {defaultValue=0.1, group=RadialNoiseMask, parameterName=RadialNoise_Tiling-X}
  n2 = ScalarParameter() {defaultValue=5, group=RadialNoiseMask, parameterName=RadialNoise_Tiling-Y}
  n3 = Constant2Vector() {g=1, r=0.5}
  n4 = TextureCoordinate() {coordinateIndex=1}
  n5 = Subtract(
   A=@n4
   B=AppendVector(
    A=ScalarParameter() {group=RadialNoiseMask, parameterName=RadialNoise_Offset-X}
    B=ScalarParameter() {defaultValue=0.5, group=RadialNoiseMask, parameterName=RadialNoise_Offset-Y}
   )
  )
  n6 = ScalarParameter() {defaultValue=-0.2, group=RadialNoiseMask, parameterName=RadialNoise_Panning-X}
  n7 = ScalarParameter() {defaultValue=0.05, group=RadialNoiseMask, parameterName=RadialNoise_Panning-Y}
  n8 = TextureSample!(
   UVs=Panner(
    Coordinate=MaterialFunctionCall!(X Scale=@n1, Y Scale=@n2, Center Offset=@n3, UV=@n5) {materialFunction=&/MF_RadialPanner_01}.Result
    Speed=AppendVector(A=@n6, B=@n7)
   )
   Tex
   "DDX(UVs)"=DDX(Value=@n4)
   "DDY(UVs)"=DDY(Value=@n4)
   Apply View MipBias
  ) {mipValueMode=TMVM_Derivative, samplerType=SAMPLERTYPE_Grayscale, texture=/Game/UI/Frontend/ItemShop/Art/Rumble/LowResBlurredNoise_Rumble}
  n9 = Multiply(A=@n8.RGB, B=ScalarParameter() {defaultValue=1, parameterName=RadialNoise_Hardness})
  LinearInterpolate(
   A=Constant() {r=1}
   B=Power(Base=Clamp(None=@n9), Exp=ScalarParameter() {defaultValue=2, parameterName=RadialNoise_Exp})
   Alpha=ScalarParameter() {group=RadialNoiseMask, parameterName=UseRadialNoiseMask}
  )
  n10 = ParticleColor()
  MaterialFunctionCall(Emissive=@n10.RGB) {materialFunction=&MF_FX_EmissiveAdjustments}
  n11 = ScalarParameter() {defaultValue=1, parameterName=Rows}
  n12 = ScalarParameter() {defaultValue=4, parameterName=Columns}
  n13 = ParticleRandom()
  n14 = TextureSample!(
   UVs=StaticSwitchParameter(
    True=MaterialFunctionCall!(UseTexCoord 1, Rows=@n11, Columns=@n12, Particle Random Value=@n13) {materialFunction=/Game/Effects/Fort_Effects/Materials/_Functions/MF_RandomSubUV_GPU_Sprites}.Result
    False=TextureCoordinate()
   ) {parameterName=Use GPU Random}
   Tex
   Apply View MipBias
  ) {texture=/Game/Effects/Fort_Effects/Textures/Mask/T_Flare_LightShafts}
  n15 = OneMinus(None=ComponentMask(None=TextureCoordinate()) {g=true})
  n16 = Saturate(None=Multiply(A=Power(Base=@n15) {constExponent=6}) {constB=51})
  n17 = Multiply(A=@n10.A, B=LinearInterpolate(A=@n14.G, B=Multiply(A=@n14.G, B=@n16), Alpha=ScalarParameter() {parameterName=Switch_AlphaGradient}))
  n18 = DepthFade(Opacity=@n17, FadeDistance=ScalarParameter() {defaultValue=100, parameterName=DepthFade})
  n19 = ScalarParameter() {defaultValue=1000, parameterName=CameraFadeLength}
  n20 = ScalarParameter() {parameterName=CameraFadeOffset}
  n21 = StaticSwitchParameter(
   True=Multiply(
    A=@n18
    B=MaterialFunctionCall(Fade Length=@n19, Fade Offset=@n20) {materialFunction=&CameraDepthFade}.Result
   )
   False=@n18
  ) {parameterName=UseCameraFade}
  n22 = VertexNormalWS()
  n23 = CameraVectorWS()
  n24 = Power(Base=Saturate(None=DotProduct(A=@n22, B=@n23)), Exp=ScalarParameter() {defaultValue=5, parameterName=ViewAnglePower})
  ShadingPathSwitch(Default=LinearInterpolate(A=@n21, B=Multiply(A=@n21, B=@n24), Alpha=ScalarParameter() {parameterName=UseAngleFade}), Mobile=@n17) {desc=Migrated from FeatureLevelSwitch}
 ]
}
