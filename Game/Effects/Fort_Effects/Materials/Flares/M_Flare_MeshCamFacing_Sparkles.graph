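# Additive unlit mesh-particle material that draws twinkling, camera-facing sparkle flares.
# World position offset billboards the mesh via the Sprite function, adding noise wobble; UV1.G and dynamic parameters vary size and motion, and near-camera fade shrinks both.
# Emissive multiplies panning star textures in two layers inside a soft radial mask, with distance-adjusted tiling, particle color and HDR boost.
# Multiplied sine waves, blended by FlickerFade, drive flicker alongside particle alpha and camera fade.
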Material /Game/Effects/Fort_Effects/Materials/Flares/M_Flare_MeshCamFacing_Sparkles
graph {
 properties {
  bUsedWithMeshParticles=true
  bUsedWithNanite=true
  blendMode=BLEND_Additive
  pixelDepthOffsetMode=PDOM_Legacy
  shadingModel=MSM_Unlit
  twoSided=true
  usedShadingModels=Unlit
 }
 nodes [
  n1 = DynamicParameter() {defaultValue=<rgba 1 1 0 1>, paramNames [FlickerFade,Random,WobbleMult,Size]}
  n2 = ComponentMask(None=TextureCoordinate() {coordinateIndex=1}) {g=true}
  n3 = MaterialFunctionCall(Distance=ScalarParameter() {defaultValue=20000, parameterName=NearCameraShrink}) {materialFunction=&Near_Camera_Fade}
  n4 = Multiply(
   A=Multiply(A=ScalarParameter() {defaultValue=1000, parameterName=SpriteScale}, B=LinearInterpolate(Alpha=@n2) {constA=1, constB=0.25})
   B=@n3.Result
  )
  n5 = Multiply(A=LinearInterpolate(Alpha=@n2) {constA=1, constB=1.5}, B=Add(A=@n1.Random, B=Time()))
  n6 = ScalarParameter() {defaultValue=0.1, parameterName=WobbleSpeed}
  n7 = TextureSample(UVs=Panner(Coordinate=TextureCoordinate() {coordinateIndex=1, uTiling=0, vTiling=6}, Time=Multiply(A=@n5, B=@n6)) {speedY=0.2}) {samplerType=SAMPLERTYPE_LinearColor, texture=&HighContrastNoise}
  n8 = Multiply(A=Multiply(A=Subtract(A=@n7.RGB) {constB=0.5}) {constB=2}, B=ScalarParameter() {defaultValue=1200, parameterName=WobbleStrength})
  Add(
   A=MaterialFunctionCall!(Center Location=WorldPosition().XYZ, 0-1 UVs, XY scale=Multiply(A=@n1.Size, B=@n4), OverridePivot, "Normals (Optional)") {materialFunction=&Sprite}.World position offset
   B=Multiply(A=@n3.Result, B=Multiply(A=@n1.WobbleMult, B=@n8))
  )
  n9 = Constant() {r=0.5}
  n10 = SphereMask(A=TextureCoordinate(), B=@n9, Radius=@n9) {hardnessPercent=0}
  n11 = Power(Base=MaterialFunctionCall(Distance=ScalarParameter() {defaultValue=80000, parameterName=TextureScaleDist}) {materialFunction=&Near_Camera_Fade}.Result) {constExponent=0.55}
  n12 = ComponentMask(None=TextureCoordinate() {coordinateIndex=1}) {g=true}
  n13 = DynamicParameter() {defaultValue=<rgba 1 1 0 1>, paramNames [FlickerFade,Random,WobbleMult,Size]}
  n14 = Multiply(A=Add(A=@n12, B=@n13.Random)) {constB=2}
  n15 = Add(A=Multiply(A=TextureCoordinate() {uTiling=0.15, vTiling=0.15}, B=LinearInterpolate(Alpha=@n11) {constA=1, constB=0.04}), B=@n14) {constB=0}
  n16 = Time()
  n17 = Add(A=@n14, B=@n16)
  n18 = TextureSample(UVs=Panner(Coordinate=@n15, Time=@n17) {speedX=0.01, speedY=-0.02}) {samplerType=SAMPLERTYPE_LinearColor, texture=/Game/Packages/Fortress_Sky/SkyDome/Master/T_SkyDome_Stars_Sharp}
  n19 = Add(A=Multiply(A=LinearInterpolate(Alpha=@n12) {constA=1, constB=-1}, B=@n16), B=@n13.Random)
  n20 = TextureSample(UVs=Panner(Coordinate=@n15, Time=@n19) {speedX=0.04, speedY=-0.05}) {samplerType=SAMPLERTYPE_Masks, texture=/Game/Characters/Player/Male/Medium/Bodies/M_Med_Soldier_04/Skins/Galaxy/Textures/Celestial_T_Stars}
  n21 = Multiply(A=@n15) {constB=1.5}
  n22 = TextureSample(UVs=Panner(Coordinate=@n21, Time=@n17) {speedX=-0.01, speedY=-0.02}) {samplerType=SAMPLERTYPE_LinearColor, texture=&T_SkyDome_Stars_Sharp}
  n23 = TextureSample(UVs=Panner(Coordinate=@n21, Time=@n19) {speedX=-0.1, speedY=0.1}) {samplerType=SAMPLERTYPE_Masks, texture=&Celestial_T_Stars}
  n24 = Multiply(A=Add(A=Multiply(A=@n18.B, B=@n20.G), B=Multiply(A=@n22.B, B=@n23.G))) {constB=6}
  n25 = ParticleColor()
  n26 = Multiply(A=Multiply(A=Power(Base=@n10), B=Clamp(None=@n24) {maxDefault=2}), B=@n25.RGB)
  MaterialFunctionCall(Emissive=MaterialFunctionCall(LDR Input=Clamp(None=@n26), HDR Tint=@n26, HDR Intensity Multiplier=ScalarParameter() {defaultValue=15, parameterName=HDRMult}) {materialFunction=&DeriveHDRfromLDR}.RemappedRange) {materialFunction=&MF_FX_EmissiveAdjustments}
  n27 = Sine(None=@n17) {period=0.155}
  n28 = Sine(None=@n17) {period=0.2}
  n29 = Multiply(A=Sine(None=@n17) {period=0.075}, B=Multiply(A=Multiply(A=@n27, B=@n28), B=Sine(None=@n17) {period=0.43}))
  Multiply(A=LinearInterpolate(A=Clamp(None=@n29) {maxDefault=3}, Alpha=@n1.FlickerFade) {constA=1}, B=Multiply(A=@n25.A, B=@n3.Result))
 ]
}
