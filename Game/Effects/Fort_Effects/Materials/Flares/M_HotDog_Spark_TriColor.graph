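# Additive unlit sprite material for spark particles, drawing a soft flare texture as a rounded "hot dog" streak in one of three colors.
# Particle random selects among particle color and two hue-shifted variants.
# UVs stretch only the middle to keep end caps round, using particle size adjusted for camera angle to particle direction.
# Opacity uses particle alpha with near-camera and depth fades; mobile skips the UV remap and both fades.
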Material /Game/Effects/Fort_Effects/Materials/Flares/M_HotDog_Spark_TriColor
graph {
 properties {bUsedWithParticleSprites=true, blendMode=BLEND_Additive, pixelDepthOffsetMode=PDOM_Legacy, shadingModel=MSM_Unlit, usedShadingModels=Unlit}
 nodes [
  n1 = ParticleColor()
  n2 = MaterialFunctionCall(Hue Shift Percentage=ScalarParameter() {defaultValue=0.1, parameterName=HueShift1}, Texture=@n1.RGB) {materialFunction=&HueShift}
  n3 = MaterialFunctionCall(Hue Shift Percentage=ScalarParameter() {defaultValue=0.6, parameterName=HueShift1_1}, Texture=@n1.RGB) {materialFunction=&HueShift}
  n4 = Constant() {r=3}
  n5 = Clamp(None=ParticleRandom()) {maxDefault=0.99, minDefault=0.01}
  n6 = Divide(A=Round(None=Multiply(A=@n4, B=@n5)), B=Constant() {r=3})
  n7 = Constant2Vector() {g=0.1, r=1}
  n8 = Abs(None=DotProduct(A=CameraVectorWS(), B=ParticleDirection()))
  n9 = Normalize(VectorInput=Multiply(A=ParticleSize(), B=LinearInterpolate(B=@n7, Alpha=@n8) {constA=1}))
  n10 = ComponentMask(None=@n9) {g=true}
  n11 = ComponentMask(None=@n9) {r=true}
  n12 = Clamp(None=Multiply(A=Divide(A=@n10, B=@n11), B=@n10) {constB=2}, Min=@n11) {maxDefault=999}
  n13 = Multiply(A=AppendVector(A=Constant() {r=1}, B=@n12), B=TextureCoordinate())
  n14 = Add(A=TextureCoordinate() {vTiling=0}, B=Constant2Vector() {g=0.5})
  n15 = ComponentMask(None=Clamp(None=@n13)) {g=true}
  n16 = Add(A=@n13, B=AppendVector(A=Constant(), B=OneMinus(None=@n12)))
  n17 = Clamp(None=@n16)
  n18 = ShadingPathSwitch(
   Default=LinearInterpolate(A=LinearInterpolate(A=@n13, B=@n14, Alpha=@n15) {constAlpha=0, desc=Blend top and middle}, B=@n16, Alpha=ComponentMask(None=@n17) {g=true}) {desc=Blend bottom}
   Mobile=TextureCoordinate() {desc=No hotdogs for mobile}
  ) {desc=Migrated from FeatureLevelSwitch}
  MaterialFunctionCall(Emissive=Multiply(
    A=MaterialFunctionCall(A=@n1.RGB, B=@n2.Result, C=@n3.Result, Alpha=@n6) {materialFunction=&3ColorBlend}.Result
    B=TextureSampleParameter2D(UVs=@n18) {parameterName=RadialGradient, texture=/Game/Global/Textures/Default/Masks_and_Gradients/T_New_Soft_Flair}.R
   )) {materialFunction=&MF_FX_EmissiveAdjustments}
  n19 = ParticleColor()
  n20 = MaterialFunctionCall() {materialFunction=&Near_Camera_Fade}
  ShadingPathSwitch(
   Default=DepthFade(Opacity=Multiply(A=@n19.A, B=@n20.Result), FadeDistance=ScalarParameter() {defaultValue=25, parameterName=DepthFadeDist})
   Mobile=@n19.A
  ) {desc=Migrated from FeatureLevelSwitch}
 ]
}
