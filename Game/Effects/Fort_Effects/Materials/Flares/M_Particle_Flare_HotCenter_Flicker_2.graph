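# Unlit additive particle flare material with a flickering hot center, built for Cascade and Niagara sprites.
# Sharpened radial masks and a spikeball flare texture form the tinted emissive, passed through MF_FX_EmissiveAdjustments.
# Multiplied sines drive flicker, controlled by particle random or Dynamic Parameters via a GPU Particles switch, with constant mobile fallback.
# Other branches build camera-distance sprite scaling, ripple and rotation offsets, and vertex-color-selected sine pulses.
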Material /Game/Effects/Fort_Effects/Materials/Flares/M_Particle_Flare_HotCenter_Flicker_2
graph {
 properties {
  bAllowNegativeEmissiveColor=true
  bUsedWithNiagaraSprites=true
  bUsedWithParticleSprites=true
  blendMode=BLEND_Additive
  pixelDepthOffsetMode=PDOM_Legacy
  shadingModel=MSM_Unlit
  usedShadingModels=Unlit
 }
 nodes [
  n1 = ScalarParameter() {defaultValue=500, parameterName=SoftenDistance}
  n2 = ScalarParameter() {defaultValue=1250, parameterName=Camera Size}
  n3 = Multiply(
   A=LinearInterpolate(
    A=ScalarParameter() {defaultValue=15, parameterName=SpriteScaleNear}
    B=ScalarParameter() {defaultValue=9, parameterName=SpriteScaleFar}
    Alpha=MaterialFunctionCall(Distance=@n1) {materialFunction=&Near_Camera_Fade}.Result
   )
   B=OneMinus(None=MaterialFunctionCall(Distance=@n2) {materialFunction=&Near_Camera_Fade}.Result)
  )
  n4 = MaterialFunctionCall!(Center Location=WorldPosition().XYZ, 0-1 UVs, XY scale=AppendVector(A=@n3, B=@n3), OverridePivot, "Normals (Optional)") {materialFunction=&Sprite}
  n5 = Panner(Coordinate=TextureCoordinate() {coordinateIndex=1, uTiling=0.5, vTiling=0.5}) {speedX=0.03, speedY=-0.06}
  Add(
   A=@n4.World position offset
   B=Multiply(
    A=TextureSample(UVs=@n5) {samplerType=SAMPLERTYPE_Normal, texture=/Game/Effects/Fort_Effects/Gameplay/Water/Textures/T_WaterRipples_01}.RGB
    B=ScalarParameter() {defaultValue=60}
   )
  )
  n6 = Time()
  Add(
   A=@n4.World position offset
   B=RotateAboutAxis(
    NormalizedRotationAxis=Constant3Vector() {constant=<rgba 0.25 0.3 1 0>}
    RotationAngle=Multiply(A=@n6) {constB=0.005}
    PivotPoint=ObjectPositionWS()
    Position=WorldPosition().XYZ
   )
  )
  n7 = ParticleRandom()
  n8 = DynamicParameter() {paramNames [FlickerFade,FlickerSpeed,Param3,Param4]}
  n9 = Multiply(
   A=Time()
   B=StaticSwitchParameter(True=LinearInterpolate(Alpha=@n7) {constA=1.2, constB=0.8}, False=@n8.FlickerSpeed) {parameterName=GPU Particles}
  )
  n10 = Sine(None=@n9) {period=0.075}
  n11 = Multiply(A=Multiply(A=Sine(None=@n9) {period=0.155}, B=Sine(None=@n9) {period=0.2}), B=Sine(None=@n9) {period=0.43})
  n12 = ScalarParameter() {defaultValue=0.4, desc="0 = full flicker amount", parameterName=GPUFlickerFadeMin}
  n13 = ScalarParameter() {defaultValue=1, desc="1 = no flicker", parameterName=GPUFlickerFadeMax, sortPriority=1}
  n14 = ParticleRandom()
  n15 = LinearInterpolate(
   A=Clamp(None=Multiply(A=@n10, B=@n11)) {maxDefault=3}
   Alpha=StaticSwitchParameter(True=LinearInterpolate(A=@n12, B=@n13, Alpha=@n14) {constA=0.5}, False=@n8.FlickerFade) {parameterName=GPU Particles}
  ) {constA=1}
  n16 = Constant() {r=0.8}
  FeatureLevelSwitch(Default=@n15, ES2=@n16, ES3_1=@n16, SM4_REMOVED=@n15)
  n17 = Constant() {r=0.5}
  n18 = SphereMask(A=TextureCoordinate(), B=@n17, Radius=@n17) {hardnessPercent=0}
  n19 = Multiply(A=Power(Base=@n18) {constExponent=25}) {constB=2}
  n20 = VectorParameter() {defaultValue=<rgba 0.201 3.0585 6 1>}
  n21 = LinearInterpolate(B=@n20.RGB) {constA=1, constAlpha=0.7}
  n22 = Add(
   A=Multiply(A=Power(Base=@n18) {constExponent=4}) {constB=0.15}
   B=TextureSampleParameter2D() {parameterName=FlareMask, texture=/Game/Effects/Fort_Effects/Textures/Flares/T_Flare_Spikeball}.RGB
  )
  MaterialFunctionCall(Emissive=Add(A=Multiply(A=@n19, B=@n21), B=Multiply(A=@n22, B=@n20.RGB))) {materialFunction=&MF_FX_EmissiveAdjustments}
  n23 = VertexColor()
  n24 = Sine(None=Multiply(A=Time()) {constB=0.4})
  n25 = Multiply(A=@n23.R, B=Multiply(A=Add(A=@n24)) {constB=0.5})
  n26 = Time()
  n27 = Sine(None=Multiply(A=Add(A=@n26) {constB=10}) {constB=0.5})
  n28 = Multiply(A=@n23.G, B=Multiply(A=Add(A=@n27)) {constB=0.5})
  n29 = Multiply(A=Add(A=Time()) {constB=150}) {constB=0.55}
  n30 = Multiply(A=Add(A=Sine(None=@n29))) {constB=0.5}
  Multiply(
   A=StaticSwitchParameter(True=@n15, False=Constant() {r=1}) {defaultValue=true, parameterName=UseFlicker}
   B=Add(A=Add(A=@n25, B=@n28), B=Multiply(A=@n23.B, B=@n30))
  )
 ]
}
