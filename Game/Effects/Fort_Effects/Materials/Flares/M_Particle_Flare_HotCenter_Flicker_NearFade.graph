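# Additive unlit particle material for a flickering lens-flare sprite with a hot center and near-camera fade.
# A powered radial mask forms a pale tinted core, added to a soft glow and FlareMask texture tinted by particle color.
# Four multiplied sines drive flicker, with speed and strength from dynamic parameters or per-particle randoms via a GPU Particles switch.
# Flicker multiplies particle-alpha depth fade and camera fade; mobile skips depth fade and flicker blending.
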Material /Game/Effects/Fort_Effects/Materials/Flares/M_Particle_Flare_HotCenter_Flicker_NearFade
graph {
 properties {
  bUsedWithNiagaraSprites=true
  bUsedWithParticleSprites=true
  blendMode=BLEND_Additive
  pixelDepthOffsetMode=PDOM_Legacy
  shadingModel=MSM_Unlit
  usedShadingModels=Unlit
 }
 nodes [
  Constant() {r=1}
  n1 = Constant() {r=0.5}
  n2 = SphereMask(A=TextureCoordinate(), B=@n1, Radius=@n1) {hardnessPercent=0}
  n3 = Multiply(A=Power(Base=@n2) {constExponent=25}) {constB=2}
  n4 = ParticleColor()
  n5 = LinearInterpolate(B=@n4.RGB) {constA=1, constAlpha=0.7}
  n6 = Add(
   A=Multiply(A=Power(Base=@n2) {constExponent=4}) {constB=0.15}
   B=TextureSampleParameter2D() {parameterName=FlareMask, texture=/Game/Effects/Fort_Effects/Textures/Flares/T_Flare_Spikeball}.RGB
  )
  MaterialFunctionCall(Emissive=Add(A=Multiply(A=@n3, B=@n5), B=Multiply(A=@n6, B=@n4.RGB))) {materialFunction=&MF_FX_EmissiveAdjustments}
  n7 = ParticleRandom()
  n8 = DynamicParameter() {paramNames [FlickerFade,FlickerSpeed,Param3,Param4]}
  n9 = Multiply(
   A=Time()
   B=StaticSwitchParameter(True=LinearInterpolate(Alpha=@n7) {constA=1.2, constB=0.8}, False=@n8.FlickerSpeed) {parameterName=GPU Particles}
  )
  n10 = Multiply(A=Sine(None=@n9) {period=0.155}, B=Sine(None=@n9) {period=0.2})
  n11 = Sine(None=@n9) {period=0.43}
  n12 = Clamp(None=Multiply(A=Sine(None=@n9) {period=0.075}, B=Multiply(A=@n10, B=@n11))) {maxDefault=3}
  n13 = ScalarParameter() {defaultValue=0.4, desc="0 = full flicker amount", parameterName=GPUFlickerFadeMin}
  n14 = ScalarParameter() {defaultValue=1, desc="1 = no flicker", parameterName=GPUFlickerFadeMax, sortPriority=1}
  n15 = ParticleRandom()
  n16 = LinearInterpolate(
   A=@n12
   Alpha=StaticSwitchParameter(True=LinearInterpolate(A=@n13, B=@n14, Alpha=@n15) {constA=0.5}, False=@n8.FlickerFade) {parameterName=GPU Particles}
  ) {constA=1}
  n17 = DepthFade(Opacity=@n4.A, FadeDistance=ScalarParameter() {defaultValue=30, parameterName=DepthFade})
  n18 = Power(Base=MaterialFunctionCall(Distance=ScalarParameter() {defaultValue=1000, parameterName=NearFade}) {materialFunction=&Near_Camera_Fade}.Result) {constExponent=16}
  Multiply(
   A=Multiply(
    A=ShadingPathSwitch(Default=@n16, Mobile=@n12) {desc=Migrated from FeatureLevelSwitch}
    B=ShadingPathSwitch(Default=@n17, Deferred=@n17, Mobile=@n4.A) {desc=Migrated from FeatureLevelSwitch}
   )
   B=LinearInterpolate(A=Constant(), B=Constant() {r=1}, Alpha=Clamp(None=@n18))
  )
 ]
}
