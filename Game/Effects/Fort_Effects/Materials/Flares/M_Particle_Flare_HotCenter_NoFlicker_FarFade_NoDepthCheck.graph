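# Unlit translucent particle and Niagara sprite material for a flare with a soft hot center, drawn without depth testing.
# A zero-hardness sphere mask centered on the UVs is raised to a power, dimmed, and added to the FlareMask texture, by default a spikeball flare.
# That sum is multiplied by particle alpha, while particle RGB passes through MF_FX_EmissiveAdjustments for emissive.
# Despite the name, the graph itself has no distance-fade nodes.
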Material /Game/Effects/Fort_Effects/Materials/Flares/M_Particle_Flare_HotCenter_NoFlicker_FarFade_NoDepthCheck
graph {
 properties {
  allowTranslucentCustomDepthWrites=true
  bDisableDepthTest=true
  bUseTranslucencyVertexFog=false
  bUsedWithNiagaraSprites=true
  bUsedWithParticleSprites=true
  blendMode=BLEND_Translucent
  pixelDepthOffsetMode=PDOM_Legacy
  shadingModel=MSM_Unlit
  usedShadingModels=Unlit
 }
 nodes [
  n1 = ParticleColor()
  MaterialFunctionCall(Emissive=@n1.RGB) {materialFunction=&MF_FX_EmissiveAdjustments}
  n2 = Constant() {r=0.5}
  n3 = Power(Base=SphereMask(A=TextureCoordinate(), B=@n2, Radius=@n2) {hardnessPercent=0}) {constExponent=4}
  Multiply(
   A=@n1.A
   B=Add(
    A=Multiply(A=@n3) {constB=0.15}
    B=TextureSampleParameter2D() {parameterName=FlareMask, texture=/Game/Effects/Fort_Effects/Textures/Flares/T_Flare_Spikeball}.RGB
   )
  )
 ]
}
