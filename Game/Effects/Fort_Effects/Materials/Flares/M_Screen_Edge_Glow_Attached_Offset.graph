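# Additive unlit, two-sided particle mesh material with no depth test that draws a directional glow toward a point offset from the actor.
# World position offset recenters the mesh on the camera and throws it far away beyond the fade distance or when facing nearly directly away.
# Camera-vector alignment shapes the glow, with power and brightness driven by distance.
# Opacity fades when looking at the source; particle color tints it, and a dynamic parameter scales fade distance.
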Material /Game/Effects/Fort_Effects/Materials/Flares/M_Screen_Edge_Glow_Attached_Offset
graph {
 properties {
  bDisableDepthTest=true
  bUsedWithMeshParticles=true
  bUsedWithNiagaraMeshParticles=true
  blendMode=BLEND_Additive
  pixelDepthOffsetMode=PDOM_Legacy
  shadingModel=MSM_Unlit
  twoSided=true
  usedShadingModels=Unlit
 }
 nodes [
  n1 = WorldPosition()
  n2 = Add(A=CameraPositionWS(), B=@n1.XYZ)
  n3 = Multiply(A=CameraPositionWS(), B=VectorParameter() {parameterName=OffsetMultiplier, defaultValue=<rgba 1 1 1 0>}.RGB)
  n4 = Add(
   A=Multiply(A=ActorPositionWS(), B=VectorParameter() {parameterName=OffsetMultiplier, defaultValue=<rgba 1 1 1 0>}.RGB)
   B=VectorParameter() {parameterName=Offset, defaultValue=<rgba 0 0 5000 0>}.RGB
  )
  n5 = DynamicParameter() {paramNames [FadeOutDistanceMult, Param2, Param3, Param4]}
  n6 = ScalarParameter() {defaultValue=150000, parameterName=FadeDistance}
  n7 = Saturate(None=Divide(A=Distance(A=@n3, B=@n4), B=Multiply(A=@n5.FadeOutDistanceMult, B=@n6)))
  n8 = Floor(None=Multiply(A=@n7) {constB=1.01})
  n9 = Normalize(VectorInput=Subtract(A=@n3, B=@n4))
  n10 = Transform(None=Constant3Vector() {constant=<rgba 0 0 1 0>}) {transformSourceType=TRANSFORMSOURCE_Camera}
  n11 = Multiply(A=OneMinus(None=DotProduct(A=@n9, B=@n10)), B=ScalarParameter() {defaultValue=3, parameterName=PopOutThreshold}) {constB=2}
  n12 = OneMinus(None=Saturate(None=Floor(None=@n11)))
  Add(
   A=Subtract(A=Subtract(A=@n2, B=@n1.XYZ), B=ActorPositionWS())
   B=Multiply(A=Add(A=@n8, B=@n12), B=Constant3Vector() {constant=<rgba 0 0 -54654566400 0>})
  )
  Constant() {r=1}
  n13 = ParticleColor()
  n14 = VectorParameter() {parameterName=DarkColor, defaultValue=<rgba 1 0 0 0>}
  n15 = VectorParameter() {parameterName=BrightColor, defaultValue=<rgba 2 1.0609 0 0>}
  n16 = CameraVectorWS()
  n17 = Power(
   Base=Saturate(None=DotProduct(A=@n16, B=@n9))
   Exp=LinearInterpolate(
    A=ScalarParameter() {defaultValue=1.5, parameterName=MinEdgeGlowPower}
    B=ScalarParameter() {defaultValue=8, parameterName=MaxEdgeGlowPower}
    Alpha=@n7
   )
  )
  n18 = Multiply(A=OneMinus(None=@n7), B=@n17)
  n19 = Saturate(None=@n18)
  MaterialFunctionCall(Emissive=Multiply(A=@n13.RGB, B=LinearInterpolate(A=@n14.RGB, B=@n15.RGB, Alpha=@n19))) {materialFunction=&MF_FX_EmissiveAdjustments}
  n20 = Multiply(A=@n9) {constB=-1}
  n21 = Power(Base=OneMinus(None=DotProduct(A=@n20, B=@n10)), Exp=ScalarParameter() {defaultValue=3, parameterName=GlowFacingFadePower})
  Multiply(A=@n13.A, B=Multiply(A=@n18, B=Multiply(A=@n17, B=@n21)))
 ]
}
