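# Unlit translucent, two-sided effects material that animates mesh quads as moving, looping sparkle sprites.
# World offset sums Sprite-function billboarding, normal expansion, directional travel and scrolling vector noise.
# A looping time, offset by world position and vertex color, drives travel and curve-atlas size over life.
# Color comes from parameters or vertex color with optional hue shift, masked by flare texture or radial SphereMask.
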Material /Game/Effects/Fort_Effects/Materials/Flares/M_Sparkles_Movement_Mesh_Only
graph {
 properties {
  bCastRayTracedShadows=false
  bEnableResponsiveAA=true
  blendMode=BLEND_Translucent
  opacityMaskClipValue=0.01
  pixelDepthOffsetMode=PDOM_Legacy
  shadingModel=MSM_Unlit
  twoSided=true
  usedShadingModels=Unlit
 }
 nodes [
  n1 = ScalarParameter() {defaultValue=20, group=Scale - Mesh Sprite, parameterName=Sprite Size}
  n2 = AppendVector(A=@n1, B=@n1)
  n3 = Time()
  n4 = MaterialFunctionCall(Float3=Multiply(A=WorldPosition().XYZ, B=ScalarParameter() {defaultValue=0.02, parameterName=WPO Divisor})) {materialFunction=&BreakOutFloat3Components}
  n5 = Add(A=Add(A=@n4.R, B=@n4.R), B=@n4.B)
  n6 = LinearInterpolate(
   A=Multiply(A=Add(A=@n3, B=@n5), B=ScalarParameter() {defaultValue=1, group=Time, parameterName="Time Multiplier (Speed)"})
   B=Multiply(A=VertexColor(), B=VectorParameter() {group=Time, parameterName="Time Modulation By Vertex Color ", defaultValue=<rgba 0 0 1 0>}.RGB)
   Alpha=ScalarParameter() {group=Time, parameterName=Modulate Speed By Vertex Color}
  )
  n7 = ChannelMaskParameter(None=VertexColor()) {group=Time, maskChannel=Blue, parameterName=Offset Time By Vertex Color Channel Mask, defaultValue=<rgba 0 0 1 0>}
  n8 = ComponentMask(None=Frac(None=Add(A=@n6, B=@n7.R))) {b=true, g=true, r=true}
  n9 = ChannelMaskParameter(None=@n8) {group=Time, maskChannel=Blue, parameterName=Offset Time By Vertex Color Channel Mask, defaultValue=<rgba 0 0 1 0>}
  n10 = CurveAtlasRowParameter(CurveTime=@n9.R) {
   atlas=/Game/Athena/LinearColorCurves/VFX_CurveAtlas
   curve=/Game/Athena/LinearColorCurves/Curve_Burst_Up_Slow_Down
   defaultValue=5
   group=Scale - Mesh Sprite
   parameterName="Sprite Size Over Life Curve (Expects Greyscale)"
  }
  n11 = LinearInterpolate(
   A=@n2
   B=Multiply(A=@n2, B=@n10.G)
   Alpha=ScalarParameter() {defaultValue=1, group=Scale - Mesh Sprite, parameterName=Modulate Size Over Life}
  )
  n12 = VertexColor()
  n13 = MaterialFunctionCall!(
   Center Location=WorldPosition().XYZ
   0-1 UVs
   XY scale=LinearInterpolate(A=@n11, B=Multiply(A=@n12.A, B=@n11), Alpha=ScalarParameter() {group=Scale - Mesh Sprite, parameterName=Modulate Size By Alpha})
   OverridePivot
   "Normals (Optional)"
  ) {materialFunction=&Sprite}
  n14 = WorldPosition()
  n15 = ObjectPositionWS()
  n16 = VertexNormalWS()
  n17 = LinearInterpolate(
   A=Normalize(VectorInput=Subtract(A=@n14.XYZ, B=@n15))
   B=Multiply(A=Normalize(VectorInput=@n16), B=ScalarParameter() {defaultValue=1, group=Scale - Expansion, parameterName=Expansion Intensity})
   Alpha=ScalarParameter() {defaultValue=1, group=Scale - Expansion, parameterName=Expansion Lerp}
  )
  n18 = Multiply(A=@n8, B=ScalarParameter() {defaultValue=25, group=Movement - Magnitude And Vector, parameterName=Magnitude Multiplier}) {constB=25}
  n19 = VectorParameter() {group=Movement - Magnitude And Vector, parameterName="Movement Overall Direction (WorldSpace)", defaultValue=<rgba 0 0 1 0>}
  n20 = WorldPosition()
  n21 = ActorPositionWS()
  n22 = ScalarParameter() {defaultValue=15, group=Movement - Noise, parameterName=Noise Speed}
  n23 = Time()
  n24 = Divide(
   A=Subtract(A=Subtract(A=@n20.XYZ, B=@n21), B=Multiply(A=@n22, B=@n23))
   B=ScalarParameter() {defaultValue=5000, group=Movement - Noise, parameterName=Noise Scale}
  )
  Add(
   A=Add(A=Add(A=@n13.World position offset, B=@n17), B=Multiply(A=@n18, B=@n19.RGB))
   B=Multiply(
    A=VectorNoise(World Position=@n24) {noiseFunction=VNF_VectorALU}
    B=ScalarParameter() {defaultValue=1, group=Movement - Noise, parameterName=Noise Intensity}
   )
  )
  n25 = ChannelMaskParameter(None=@n8) {group=Time, maskChannel=Blue, parameterName=Offset Time By Vertex Color Channel Mask, defaultValue=<rgba 0 0 1 0>}
  n26 = Sine(None=Multiply(A=VertexColor().A, B=@n25.R))
  n27 = Clamp(None=VertexInterpolator(VS=Abs(None=@n26)).PS) {minDefault=0.5}
  Clamp(None=Multiply(A=ComponentMask(None=@n27) {r=true}) {constB=50}) {maxDefault=50}
  n28 = VertexColor()
  n29 = Multiply(A=AppendVector(A=AppendVector(A=@n28.R, B=@n28.R), B=@n28.B), B=@n28.A) {constB=0.025}
  n30 = VectorParameter() {group=Color, parameterName=Color 1, defaultValue=<rgba 1 1 1 0>}
  n31 = VectorParameter() {group=Color, parameterName=Color 2, defaultValue=<rgba 1 1 1 0>}
  n32 = ChannelMaskParameter(None=TextureSampleParameter2D() {group=Textures - Opacity, parameterName=Opacity Mask Texture, texture=/Game/Effects/Fort_Effects/Textures/Flares/T_GlowSar_6pt}.RGB) {group=Textures - Opacity, maskChannel=Blue, parameterName=Mask Texture Channel Mask, defaultValue=<rgba 0 0 1 0>}
  n33 = LinearInterpolate(
   A=LinearInterpolate(
    A=LinearInterpolate(A=@n30.RGB, B=@n31.RGB, Alpha=@n32.R)
    B=VectorParameter() {group=Color, parameterName=Color_Sparkle, defaultValue=<rgba 1 1 1 0>}.RGB
    Alpha=ScalarParameter() {defaultValue=1, group=Color, parameterName=Use Single Color Parameter}
   )
   B=@n29
   Alpha=ScalarParameter() {group=Color, parameterName=Use Vertex Colors As Color}
  )
  n34 = Clamp(None=LinearInterpolate(
    A=@n33
    B=MaterialFunctionCall(Hue Shift Percentage=@n29, Texture=@n33) {materialFunction=&HueShift}.Result
    Alpha=ScalarParameter() {group=Color - Hue Shift, parameterName=Use Hue Shift}
   )) {maxDefault=4}
  n35 = ChannelMaskParameter(None=TextureSampleParameter2D() {group=Textures - Opacity, parameterName=Opacity Mask Texture, texture=&T_GlowSar_6pt}.RGB) {group=Textures - Opacity, maskChannel=Blue, parameterName=Mask Texture Channel Mask, defaultValue=<rgba 0 0 1 0>}
  n36 = LinearInterpolate(
   A=@n35.R
   B=SphereMask(
    A=TextureCoordinate()
    B=Constant2Vector() {g=0.5, r=0.5}
    Radius=ScalarParameter() {defaultValue=0.45, parameterName=Radius}
    Hardness=ScalarParameter() {defaultValue=1, parameterName=Hardness}
   )
   Alpha=ScalarParameter() {parameterName=Use Simple Opacity}
  )
  MaterialFunctionCall(Emissive=LinearInterpolate(A=@n34, B=Multiply(A=@n34, B=@n36), Alpha=ScalarParameter() {group=Overall Settings, parameterName=Is Additive})) {materialFunction=&MF_FX_EmissiveAdjustments}
  VectorParameter() {group=Color, parameterName=Color, defaultValue=<rgba 1 1 1 0>}
 ]
}
