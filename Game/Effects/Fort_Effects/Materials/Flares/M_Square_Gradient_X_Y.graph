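# Unlit translucent FX material for sprite and mesh particles that draws a soft square gradient without textures.
# Two one-axis sphere masks on U and V multiply into the shape, with per-axis hardness driven by dynamic parameters.
# Particle color or a mesh tint is scaled by time-of-day values for emissive.
# Opacity combines particle alpha, optional view-angle fade, and near-camera and depth fade, the latter two skipped on mobile.
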Material /Game/Effects/Fort_Effects/Materials/Flares/M_Square_Gradient_X_Y
graph {
 properties {
  bUsedWithNanite=true
  bUsedWithNiagaraMeshParticles=true
  bUsedWithNiagaraSprites=true
  bUsedWithParticleSprites=true
  blendMode=BLEND_Translucent
  pixelDepthOffsetMode=PDOM_Legacy
  shadingModel=MSM_Unlit
  usedShadingModels=Unlit
 }
 nodes [
  n1 = ScalarParameter() {defaultValue=1, parameterName="Value_Morning (S)"}
  n2 = ScalarParameter() {defaultValue=1, parameterName="Value_Day (S)"}
  n3 = ScalarParameter() {defaultValue=1, parameterName="Value_Evening (S)"}
  n4 = ScalarParameter() {defaultValue=1, parameterName="Value_Night (S)"}
  n5 = ParticleColor()
  n6 = VectorParameter() {parameterName=MeshColorTint, defaultValue=<rgba 0 0 0 1>}
  n7 = ScalarParameter() {parameterName=UseMeshColorTintOrParticleColor}
  MaterialFunctionCall(Emissive=Multiply(
    A=MaterialFunctionCall(Value_Morning=@n1, Value_Day=@n2, Value_Evening=@n3, Value_Night=@n4) {materialFunction=&Scalar_Time_of_Day}.Result
    B=LinearInterpolate(A=@n5.RGB, B=@n6.RGB, Alpha=@n7)
   )) {materialFunction=&MF_FX_EmissiveAdjustments}
  n8 = TextureCoordinate()
  n9 = Constant() {r=0.5}
  n10 = DynamicParameter() {defaultValue=<rgba 0.5 0.5 1 1>, paramNames [Hardness X, Hardness Y, Param3, Param4]}
  n11 = SphereMask(A=ComponentMask(None=@n8) {r=true}, B=@n9, Radius=@n9, Hardness=@n10.Hardness X) {hardnessPercent=50}
  n12 = SphereMask(A=ComponentMask(None=@n8) {g=true}, B=@n9, Radius=@n9, Hardness=@n10.Hardness Y) {hardnessPercent=90}
  n13 = DepthFade(
   Opacity=MaterialFunctionCall() {materialFunction=&Near_Camera_Fade}.Result
   FadeDistance=ScalarParameter() {defaultValue=50, parameterName=DepthFade}
  )
  n14 = ShadingPathSwitch(Default=@n13, Deferred=@n13, Mobile=Constant() {r=1}) {desc=Migrated from FeatureLevelSwitch}
  n15 = ScalarParameter() {parameterName=AngleFade}
  Multiply(
   A=Multiply(A=Multiply(A=@n11, B=@n12), B=Multiply(A=@n5.A, B=@n14))
   B=LinearInterpolate(
    B=MaterialFunctionCall("FadeSensitivity (optional)"=@n15) {materialFunction=&FadeBasedOnViewAngle}.Result
    Alpha=ScalarParameter() {parameterName=UseFadeBasedOnViewAngle}
   ) {constA=1}
  )
 ]
}
