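# Unlit additive flare material for static meshes and Niagara mesh particles, turning them into scalable sprites that bob over time.
# A static switch selects particle color with a radial gradient, or parameter color with a spikeball texture plus sphere-mask glow.
# Opacity optionally blends toward depth fade, which mobile skips.
# World position offset sums sprite billboarding, camera offset and switchable bob, whose raw sine lerp alpha overshoots the BobHeightMax end.
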Material /Game/Effects/Fort_Effects/Materials/Flares/M_StaticMesh_Flare_W_Bob
graph {
 properties {
  bUsedWithNiagaraMeshParticles=true
  blendMode=BLEND_Additive
  pixelDepthOffsetMode=PDOM_Legacy
  shadingModel=MSM_Unlit
  usedShadingModels=Unlit
 }
 nodes [
  n1 = ParticleColor()
  n2 = VectorParameter() {group=Color, parameterName=Color, defaultValue=<rgba 1 1 1 1>}
  n3 = ScalarParameter() {defaultValue=1, parameterName=OpacityScalar}
  n4 = TextureCoordinate()
  n5 = Constant() {r=0.5}
  n6 = ScalarParameter() {group=Color, parameterName=Hardness}
  n7 = Add(
   A=TextureSample() {texture=/Game/Effects/Fort_Effects/Textures/Flares/T_Flare_Spikeball}.G
   B=Power(
    Base=SphereMask(A=@n4, B=@n5, Radius=@n5, Hardness=@n6) {hardnessPercent=0}
    Exp=ScalarParameter() {defaultValue=5, group=Color, parameterName=GrandientFalloff}
   )
  )
  n8 = StaticSwitchParameter(
   True=Multiply(A=@n1.A, B=TextureSample() {texture=/Game/Athena/HUD/Materials/T_UI_RadialGradient}.R)
   False=Multiply(A=Multiply(A=@n2.A, B=@n3), B=Saturate(None=@n7))
  ) {parameterName=Particle Color?}
  n9 = ScalarParameter() {defaultValue=100, parameterName=DepthFadeDistance}
  LinearInterpolate(
   A=@n8
   B=ShadingPathSwitch(Default=DepthFade(Opacity=@n8, FadeDistance=@n9), Mobile=@n8) {desc=Migrated from FeatureLevelSwitch}
   Alpha=ScalarParameter() {parameterName=UseDepthFade?, sliderMax=1}
  )
  MaterialFunctionCall(Emissive=StaticSwitchParameter(True=@n1.RGB, False=@n2.RGB) {parameterName=Particle Color?}) {materialFunction=&MF_FX_EmissiveAdjustments}
  n10 = ScalarParameter() {defaultValue=10, group=Scale, parameterName=ScaleX}
  n11 = ScalarParameter() {defaultValue=10, group=Scale, parameterName=ScaleY}
  n12 = Add(
   A=MaterialFunctionCall!(Center Location=ObjectPositionWS(), 0-1 UVs, XY scale=AppendVector(A=@n10, B=@n11), OverridePivot, "Normals (Optional)") {materialFunction=&Sprite}.World position offset
   B=MaterialFunctionCall(Offset Amount=ScalarParameter() {group=WPO, parameterName=CameraOffset}) {materialFunction=/Engine/Functions/Engine_MaterialFunctions02/WorldPositionOffset/CameraOffset}.Camera Offset
  )
  n13 = VectorParameter() {group=WPO, parameterName=BobHeightMax, defaultValue=<rgba 0 0 10 0>}
  n14 = VectorParameter() {group=WPO, parameterName=BobHeightMin, defaultValue=<rgba 0 0 -2 0>}
  n15 = Sine(None=Multiply(A=Time(), B=ScalarParameter() {defaultValue=0.4, group=WPO, parameterName=WPOTimeScale}))
  StaticSwitchParameter(True=Add(A=@n12, B=LinearInterpolate(A=@n13.RGB, B=@n14.RGB, Alpha=@n15)), False=@n12) {defaultValue=true, parameterName=Bob?}
 ]
}
